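# Physics asset for the M_MED_PreciseRuffian backpack cosmetic's pack mesh, giving the back accessory a simple physics shape.
# It defines a single body on the pack bone using one offset, tilted capsule.
# It has no constraints, so there are no jointed or linked bodies.
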PhysicsAsset /BRCosmetics/Accessories/FORT_Backpacks/Backpack_M_Med_PreciseRuffian/Meshes/M_MED_PreciseRuffian_Pack_PhysicsAsset
bodies {pack [{shapeType=Capsule, length=46.1386, radius=10.0841, center=<-1.642 -3.4484 -5.2372>, rotation=<rot -0.0079 -90 29.8303>}]}
constraints []
