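# Anim Blueprint for the PreciseRuffian backpack cosmetic, procedurally offsetting its pack bone as the character moves.
# Each update derives rotation and translation from the owner's spine_05-to-spine_01 delta and converts skydive, melee and jump states to alphas.
# Four chained additive Modify Bone nodes apply these offsets.
# Branches unconnected to the update event toggle a "Hide Element 01" material parameter by equipped-pickaxe tag and skydiving, with debug prints.
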AnimBlueprint /BRCosmetics/Accessories/FORT_Backpacks/Backpack_M_Med_PreciseRuffian/Meshes/M_MED_PreciseRuffian_Pack_Skeleton_AnimBlueprint
target_skeleton=&/M_MED_PreciseRuffian_Pack_Skeleton
graphs [
 AnimGraph {nodes [
   n1 = "Transform (Modify) Bone - Bone: pack"(
    ComponentPose=Local To Component(LocalPose=Input Pose())
    Translation=Get BP Translate() -> (BP Translate:Vector)
    Rotation=Get BP Rotate() -> (BP Rotate:Rotator)
   ) {
    rotationMode=BMM_Additive
    rotationSpace=BCS_BoneSpace
    translationMode=BMM_Additive
    translationSpace=BCS_BoneSpace
    alphaBoolBlend {blendOption=Linear}
    alphaScaleBiasClamp {bInterpResult=true}
    boneToModify {boneName=pack}
    rotation {roll=10}
   }
   n2 = "Transform (Modify) Bone - Bone: pack"(ComponentPose=@n1, Alpha=Get Is Skydive() -> (Is Skydive:double), Translation="-2,0,4", Rotation="-15,0,2.5") {
    rotationMode=BMM_Additive
    translationMode=BMM_Additive
    translationSpace=BCS_BoneSpace
    alphaBoolBlend {blendOption=Linear}
    alphaScaleBiasClamp {bInterpResult=true}
    boneToModify {boneName=pack}
    rotation {pitch=-15, roll=2.5}
    translation {x=-2, z=4}
   }
   n3 = "Transform (Modify) Bone - Bone: pack"(ComponentPose=@n2, Alpha=Get Is Swing() -> (Is Swing:double), Translation="0,0,3", Rotation="-8,0,0") {
    rotationMode=BMM_Additive
    translationMode=BMM_Additive
    translationSpace=BCS_BoneSpace
    alphaBoolBlend {blendOption=Linear}
    alphaScaleBiasClamp {bInterpResult=true, interpSpeedDecreasing=1}
    boneToModify {boneName=pack}
    rotation {pitch=-8}
    translation {z=3}
   }
   n4 = "Transform (Modify) Bone - Bone: pack"(ComponentPose=@n3, Alpha=Get Is Jump() -> (Is Jump:double), Translation="0,0,0", Rotation="0,5,0") {
    rotationMode=BMM_Additive
    translationMode=BMM_Additive
    translationSpace=BCS_BoneSpace
    alphaBoolBlend {blendOption=Linear}
    alphaScaleBiasClamp {bInterpResult=true, interpSpeedDecreasing=5, interpSpeedIncreasing=5}
    boneToModify {boneName=pack}
    rotation {yaw=5}
   }
   Output Pose(Result=Component To Local(ComponentPose=@n4))
  ]}
 EventGraph {nodes [
   n1 = Get Dynamic Material Instance(SkeletalMeshComponent=GetPartSkeletalMeshComponent())
   n2 = PickaxeItemDefinitionHasAnyMetaTag(Tags=Make Array("[0]"=(TagName="Cosmetics.MetaTags.PreciseRuffian")), bRequiresEquipped=true)
   > n3 = Branch(Condition=@n2)
   > n4 = Branch(Condition=Get bIsSkyDiving())
   SetScalarParameterValue(execute=[@n4.then, @n9.else, @n3.else], self=@n1, ParameterName=Hide Element 01, Value=0)
   > PrintString!(execute, InString=bye, bPrintToScreen, bPrintToLog, TextColor, Duration, Key) -> (then)
   SetScalarParameterValue(execute=[@n4.else, @n9.then], self=@n1, ParameterName=Hide Element 01, Value=1)
   > PrintString(InString=Hello)
   Event BlueprintUpdateAnimation()
   > Sequence()
   n5 = GetOwnerSkeletalMeshComponent()
   n6 = BreakRotator(InRot=BreakTransform(InTransform=GetDeltaTransformFromRefPose(self=@n5, BoneName=spine_05, BaseName=spine_01)).Rotation)
   n7 = FInterpEaseInOut!(A:double=0, B:double=MapRangeClamped(Value=@n6, InRangeA=0, InRangeB=-40, OutRangeB=15), Alpha:double=0, Exponent:double=0) -> (ReturnValue:double)
   > Set BP Rotate(BP Rotate:Rotator=MakeRotator(Roll=@n7, Pitch=@n7)) -> (Output_Get:Rotator)
   n8 = MapRangeClamped(Value=@n6, InRangeA=0, InRangeB=-10, OutRangeB=-1.5)
   > Set BP Translate(BP Translate:Vector=MakeVector(Y=FInterpEaseInOut(A=0, B=@n8, Alpha=0, Exponent=0))) -> (Output_Get:Vector)
   > Set Is Skydive(Is Skydive:double="To Float (Boolean)"(InBool=Get bIsSkyDiving())) -> (Output_Get:double)
   > Set Is Swing(Is Swing:double="To Float (Boolean)"(InBool=Get bIsPlayingMeleeAnim())) -> (Output_Get:double)
   > Set Is Jump(Is Jump:double="To Float (Boolean)"(InBool=Get bIsJumping())) -> (Output_Get:double)
   n9 = Branch(Condition=@n2)
   MakeVector(Y=0)
  ]}
]
variables [
 Body_Mesh {type=SceneComponent, default=null}
 BP Rotate {type=Rotator}
 BP Translate {type=Vector}
 Gravity Override Params {type=GravityOverrideParamsStruct}
 GravityOverride {type=Vector}
 Is Jump {type=number}
 Is Skydive {type=number}
 Is Swing {type=number}
 Skel Mesh {type=SkeletalMeshComponent}
 spine_01_location {type=Vector}
 spine_01_rotation {type=Rotator}
 spine_02_location {type=Vector}
 spine_02_rotation {type=Rotator}
 spine_03_location {type=Vector}
 spine_03_rotation {type=Rotator}
 spine_04_location {type=Vector}
 spine_04_rotation {type=Rotator}
 spineRot {type=number}
]
